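# Looping sound cue for Chrome vehicle ambience that follows the VehicleSpeed parameter, crossfading between idle and moving layers.
# A mixer blends two looping waves, each behind a continuous modulator that raises pitch with speed.
# As speed rises the idle layer gets quieter and the moving layer gets louder.
# It uses the valet movement attenuation asset, the vehicles sound class and a small concurrency cap that stops lowest priority.
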SoundCue /ChromeShared/Audio/Vehicles/Chrome_Veh_Ambient_Loop_Cue
root=Mixer_0
nodes [
 {id=Mixer_0, children [ModulatorContinuous_0, ModulatorContinuous_1], properties {inputVolume [1,1]}}
 {
  id=ModulatorContinuous_0
  children [WavePlayer_0]
  properties {
   pitchModulationParams {maxOutput=1.5, minOutput=0.9, parameterName=VehicleSpeed}
   volumeModulationParams {maxOutput=0.6, minOutput=1, parameterName=VehicleSpeed}
  }
 }
 {
  id=ModulatorContinuous_1
  children [WavePlayer_1]
  properties {
   pitchModulationParams {minOutput=0.5, parameterName=VehicleSpeed}
   volumeModulationParams {maxOutput=1.5, minInput=0.1, parameterName=VehicleSpeed}
  }
 }
 {id=WavePlayer_0, properties {bLooping=true, soundWaveAssetPtr=&/Chrome_Veh_Ambient_Idle_Loop_01}}
 {id=WavePlayer_1, properties {bLooping=true, soundWaveAssetPtr=&/Chrome_Veh_Ambient_Moving_Loop_01}}
]
properties {
 attenuationSettings=&/Chrome_Veh_Valet_Movement_Att
 bOverrideConcurrency=true
 maxDistance=12512
 soundClassObject=&Fort_Vehicles
 volumeMultiplier=0.7
 attenuationOverrides {bAttenuate=false, bSpatialize=false, falloffDistance=2400, attenuationShapeExtents {x=600}}
 concurrencyOverrides {maxCount=3, resolutionRule=StopLowestPriority}
}
