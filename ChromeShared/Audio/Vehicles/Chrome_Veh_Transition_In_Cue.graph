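# Sound cue for the Chrome vehicle transition-in, playing one of three ambient loop start sounds in the vehicles sound class.
# An equally weighted Random node, excluded from branch culling, feeds a Modulator that holds volume fixed and leaves pitch at default modulation.
# Spatialization comes from the shared Chrome valet-activate attenuation asset.
# A concurrency override caps simultaneous plays at two, stopping the lowest priority.
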SoundCue /ChromeShared/Audio/Vehicles/Chrome_Veh_Transition_In_Cue
root=Modulator_0
nodes [
 {id=Modulator_0, children [Random_0], properties {volumeMax=1, volumeMin=1}}
 {id=Random_0, children [WavePlayer_0,WavePlayer_1,WavePlayer_2], properties {bShouldExcludeFromBranchCulling=true, weights [1,1,1]}}
 {id=WavePlayer_0, properties {soundWaveAssetPtr=&/Chrome_Veh_Ambient_Loop_Start_03}}
 {id=WavePlayer_1, properties {soundWaveAssetPtr=&/Chrome_Veh_Ambient_Loop_Start_01}}
 {id=WavePlayer_2, properties {soundWaveAssetPtr=&/Chrome_Veh_Ambient_Loop_Start_02}}
]
properties {
 attenuationSettings=&/Chrome_Veh_Valet_Activate_Att
 bOverrideConcurrency=true
 maxDistance=12512
 soundClassObject=&Fort_Vehicles
 volumeMultiplier=1
 attenuationOverrides {falloffDistance=2400, attenuationShapeExtents {x=600}}
 concurrencyOverrides {maxCount=2, resolutionRule=StopLowestPriority}
}
