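# Sound cue for a Chrome vehicle transition-out sound, built from ambient loop end one-shots.
# A Random node picks one of three equally weighted end waves and feeds it through a Modulator with fixed unit volume; the Random node is excluded from branch culling.
# It uses the Valet activate attenuation asset and the vehicle sound class, with overridden concurrency that caps simultaneous plays and stops the lowest priority.
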SoundCue /ChromeShared/Audio/Vehicles/Chrome_Veh_Transition_Out_Cue
root=Modulator_0
nodes [
 {id=Modulator_0, children [Random_1], properties {volumeMax=1, volumeMin=1}}
 {id=Random_1, children [WavePlayer_3,WavePlayer_4,WavePlayer_5], properties {bShouldExcludeFromBranchCulling=true, weights [1,1,1]}}
 {id=WavePlayer_3, properties {soundWaveAssetPtr=&/Chrome_Veh_Ambient_Loop_End_01}}
 {id=WavePlayer_4, properties {soundWaveAssetPtr=&/Chrome_Veh_Ambient_Loop_End_02}}
 {id=WavePlayer_5, properties {soundWaveAssetPtr=&/Chrome_Veh_Ambient_Loop_End_03}}
]
properties {
 attenuationSettings=&/Chrome_Veh_Valet_Activate_Att
 bOverrideConcurrency=true
 maxDistance=12512
 soundClassObject=&Fort_Vehicles
 volumeMultiplier=1
 attenuationOverrides {falloffDistance=2400, attenuationShapeExtents {x=600}}
 concurrencyOverrides {maxCount=2, resolutionRule=StopLowestPriority}
}
